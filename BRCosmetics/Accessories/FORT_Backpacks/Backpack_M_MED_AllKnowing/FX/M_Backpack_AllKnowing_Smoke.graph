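# Translucent, two-sided unlit smoke material for Niagara mesh particles on the AllKnowing backpack cosmetic.
# Three distorted, panning noise layers feed a dissolve texture and warped V gradient to carve opacity, softened by view angle and non-mobile depth fade.
# A distorted detail texture blends emissive colors and opacity levels, while panning noise drives world position offset weighted along V.
# Dynamic parameters supply UV offsets, WPO intensity and dissolve range per particle.
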Material /BRCosmetics/Accessories/FORT_Backpacks/Backpack_M_MED_AllKnowing/FX/M_Backpack_AllKnowing_Smoke
graph {
 properties {
  bUsedWithNiagaraMeshParticles=true
  blendMode=BLEND_Translucent
  pixelDepthOffsetMode=PDOM_Legacy
  shadingModel=MSM_Unlit
  twoSided=true
  usedShadingModels=Unlit
 }
 nodes [
  n1 = VectorParameter() {group=Emissive, parameterName=Emissive Color B, defaultValue=<rgba 0 0 0 1>}
  n2 = VectorParameter() {group=Emissive, parameterName=Emissive Color A, defaultValue=<rgba 0.01 0.01 0.01 1>}
  n3 = TextureCoordinate()
  n4 = VectorParameter() {group=Detail, parameterName=Detail Distortion UV, sortPriority=1, defaultValue=<rgba 2 2 0 0>}
  n5 = AppendVector(A=@n4.R, B=@n4.R)
  n6 = TextureSampleParameter2D(UVs=Panner(Coordinate=Multiply(A=@n3, B=@n5), Speed=AppendVector(A=@n4.B, B=@n4.B))) {
   group=Distortion
   parameterName="Distortion  Tex"
   samplerType=SAMPLERTYPE_Normal
   sortPriority=0
   texture=&T_Weap_Distort_N_02
  }
  n7 = ComponentMask(None=@n6.RGB) {g=true, r=true}
  n8 = ScalarParameter() {defaultValue=0.1, group=Detail, parameterName=Detail Distortion Intensity}
  n9 = DynamicParameter() {defaultValue=<rgba 0 1 1 0>, paramNames [Random Offset, WPO intensity, m, ""]}
  n10 = AppendVector(A=@n9.Random Offset, B=@n9.Random Offset)
  n11 = VectorParameter() {group=Detail, parameterName=Detail UV, sortPriority=2, defaultValue=<rgba 4 0.5 0 0.5>}
  n12 = Panner(Coordinate=Multiply(A=TextureCoordinate(), B=AppendVector(A=@n11.R, B=@n11.R)), Speed=AppendVector(A=@n11.B, B=@n11.B))
  n13 = TextureSampleParameter2D(UVs=Add(A=Multiply(A=@n7, B=@n8), B=Add(A=@n10, B=@n12))) {group=Detail, parameterName=Detail Tex, texture=/Game/Global/Textures/Default/Tileable_Patterns/T_Blowing_Smoke_Pattern2_Tile_DISP}
  n14 = ParticleColor()
  MaterialFunctionCall(Emissive=Multiply(A=LinearInterpolate(A=@n1.RGB, B=@n2.RGB, Alpha=@n13.R), B=@n14.RGB)) {materialFunction=&MF_FX_EmissiveAdjustments}
  n15 = DynamicParameter() {defaultValue=<rgba 0 1 1 0>, paramNames [Random Offset, WPO intensity, m, ""]}
  n16 = VectorParameter() {group=WPO, parameterName=WPO Side UV, sortPriority=6, defaultValue=<rgba 1 1 -0.1 0.3>}
  n17 = TextureCoordinate()
  n18 = Multiply(A=@n17, B=AppendVector(A=@n16.R, B=@n16.R))
  n19 = AppendVector(A=@n16.B, B=@n16.B)
  n20 = TextureSample(UVs=Add(A=AppendVector(A=@n15.Random Offset, B=@n15.Random Offset), B=Panner(Coordinate=@n18, Speed=@n19))) {samplerType=SAMPLERTYPE_Normal, texture=/Game/Characters/Player/Female/Medium/Bodies/F_MED_Bistro_Astronaut/FX/T_GaussianNoise_Normal}
  n21 = Multiply(A=ComponentMask(None=@n20.RGB) {g=true, r=true}, B=ScalarParameter() {defaultValue=1, group=WPO, parameterName=WPO Side Intensity}) {constB=10}
  n22 = AppendVector(A=@n15.Random Offset, B=@n15.Random Offset)
  n23 = VectorParameter() {group=WPO, parameterName=WPO Wave UV, sortPriority=6, defaultValue=<rgba 1 1 0 0.5>}
  n24 = Panner(Coordinate=Multiply(A=@n17, B=AppendVector(A=@n23.R, B=@n23.R)), Speed=AppendVector(A=@n23.B, B=@n23.B))
  n25 = Multiply(
   A=TextureSample(UVs=Add(A=@n22, B=@n24)) {texture=&T_Smooth_tile}.R
   B=ScalarParameter() {defaultValue=1, group=WPO, parameterName=WPO Wave Intensity}
  ) {constB=10}
  n26 = ComponentMask(None=TextureCoordinate()) {g=true}
  Multiply(A=LinearInterpolate(B=AppendVector(A=@n21, B=@n25), Alpha=OneMinus(None=@n26)), B=@n15.WPO intensity) {constB=10}
  n27 = DynamicParameter() {defaultValue=<rgba 0 1 1 0>, paramNames [Random Offset, WPO intensity, m, ""]}
  n28 = TextureCoordinate()
  n29 = VectorParameter() {group=Noise, parameterName=Noise A UV, sortPriority=2, defaultValue=<rgba 6 0.5 0 -0.5>}
  n30 = AppendVector(A=@n29.R, B=@n29.R)
  n31 = Add(
   A=AppendVector(A=@n27.Random Offset, B=@n27.Random Offset)
   B=Panner(Coordinate=Multiply(A=@n28, B=@n30), Speed=AppendVector(A=@n29.B, B=@n29.B))
  )
  n32 = TextureCoordinate()
  n33 = VectorParameter() {group=Distortion, parameterName=Noise Distortion UV, sortPriority=1, defaultValue=<rgba 1 1.25 0 0>}
  n34 = Multiply(A=@n32, B=AppendVector(A=@n33.R, B=@n33.R))
  n35 = AppendVector(A=@n33.B, B=@n33.B)
  n36 = TextureSampleParameter2D(UVs=Add(A=AppendVector(A=@n27.Random Offset, B=@n27.Random Offset), B=Panner(Coordinate=@n34, Speed=@n35))) {
   group=Distortion
   parameterName="Distortion  Tex"
   samplerType=SAMPLERTYPE_Normal
   sortPriority=0
   texture=&T_Weap_Distort_N_02
  }
  n37 = ComponentMask(None=@n32) {g=true}
  n38 = Multiply(
   A=ComponentMask(None=@n36.RGB) {g=true, r=true}
   B=LinearInterpolate(
    B=OneMinus(None=@n37)
    Alpha=ScalarParameter() {defaultValue=1, group=Distortion, parameterName=Noise Distortion Gradient Lerp, sortPriority=3}
   ) {constA=1, constAlpha=0}
  )
  n39 = Multiply(A=ScalarParameter() {defaultValue=1, group=Distortion, parameterName=Noise A Distortion Intensity, sortPriority=2}, B=@n38)
  n40 = ChannelMaskParameter(None=TextureSampleParameter2D(UVs=Add(A=@n39, B=@n31)) {group=Noise, parameterName=NoIse A Texture, sortPriority=1, texture=&T_NoisePack_01}.RGBA) {group=Noise, parameterName=Noise A Channel, sortPriority=0}
  n41 = ScalarParameter() {defaultValue=0.5, group=Noise, parameterName=NoIse A Intensity, sortPriority=3}
  n42 = AppendVector(A=@n27.Random Offset, B=@n27.Random Offset)
  n43 = VectorParameter() {group=Noise, parameterName=Noise B UV, sortPriority=6, defaultValue=<rgba 3 1 0.1 0.4>}
  n44 = Panner(Coordinate=Multiply(A=TextureCoordinate(), B=AppendVector(A=@n43.R, B=@n43.R)), Speed=AppendVector(A=@n43.B, B=@n43.B))
  n45 = ScalarParameter() {defaultValue=0.1, group=Distortion, parameterName=Noise B Distortion Intensity, sortPriority=2}
  n46 = TextureSampleParameter2D(UVs=Add(A=Multiply(A=@n38, B=@n45), B=Add(A=@n42, B=@n44))) {
   group=Noise
   parameterName=NoIse B Texture
   sortPriority=5
   texture=/Game/Accessories/FORT_Backpacks/Backpack_M_MED_AllKnowing/FX/T_Backpack_AllKnowing_Noise_02
  }
  n47 = Multiply(
   A=ChannelMaskParameter(None=@n46.R) {group=Noise, parameterName=Noise B Channel, sortPriority=1}.R
   B=ScalarParameter() {defaultValue=3, group=Noise, parameterName=NoIse B Intensity, sortPriority=7}
  ) {constB=10}
  n48 = AppendVector(A=@n27.Random Offset, B=@n27.Random Offset)
  n49 = VectorParameter() {group=Noise, parameterName=Noise C UV, sortPriority=10, defaultValue=<rgba 3 1 -0.15 0.4>}
  n50 = Panner(Coordinate=Multiply(A=TextureCoordinate(), B=AppendVector(A=@n49.R, B=@n49.R)), Speed=AppendVector(A=@n49.B, B=@n49.B))
  n51 = ScalarParameter() {defaultValue=0.4, group=Distortion, parameterName=Noise C Distortion Intensity, sortPriority=2}
  n52 = TextureSampleParameter2D(UVs=Add(A=Multiply(A=@n38, B=@n51), B=Add(A=@n48, B=@n50))) {
   group=Noise
   parameterName=NoIse C Texture
   sortPriority=9
   texture=/Game/Accessories/FORT_Backpacks/Backpack_M_MED_AllKnowing/FX/T_Backpack_AllKnowing_Noise_01
  }
  n53 = Multiply(
   A=ChannelMaskParameter(None=@n52.RGB) {group=Noise, parameterName=Noise C Channel, sortPriority=8}.R
   B=ScalarParameter() {defaultValue=1, group=Noise, parameterName=NoIse C Intensity, sortPriority=11}
  ) {constB=10}
  n54 = LinearInterpolate(A=Saturate(None=Multiply(A=@n40.R, B=@n41) {constB=10}), B=Add(A=Saturate(None=@n47), B=Saturate(None=@n53))) {constAlpha=0.85}
  n55 = VectorParameter() {group=Dissolve, parameterName=Dissolve Tex UV, sortPriority=2, defaultValue=<rgba 2 0.1 0 0>}
  n56 = Multiply(A=TextureCoordinate(), B=AppendVector(A=@n55.R, B=@n55.R))
  n57 = AppendVector(A=@n55.B, B=@n55.B)
  n58 = TextureSampleParameter2D(UVs=Add(A=Constant2Vector() {g=0.1}, B=Panner(Coordinate=@n56, Speed=@n57))) {
   group=Dissolve
   parameterName=Dissolve Tex
   samplerType=SAMPLERTYPE_Normal
   sortPriority=0
   texture=&T_Weap_Distort_N_02
  }
  n59 = StaticComponentMaskParameter(None=@n58.RGB) {defaultG=true, parameterName=Dissolve Tex Channel}
  n60 = ScalarParameter() {defaultValue=10, group=Dissolve, parameterName=Dissolve Tex Hardness, sliderMin=0.5}
  n61 = ScalarParameter() {defaultValue=0.9, group=Dissolve, parameterName=Dissolve Tex Range}
  n62 = Multiply(A=@n54, B=Subtract(A=Multiply(A=@n59, B=@n60), B=LinearInterpolate(A=@n60, Alpha=@n61) {constB=-1}))
  n63 = ScalarParameter() {defaultValue=1, group=Dissolve, parameterName=Use Dissolve Tex}
  n64 = DynamicParameter() {parameterIndex=1, defaultValue=<rgba 0 0 0 0.1>, paramNames ["", "", Gradient Distortion Offset U, Gradient Distortion Offset V]}
  n65 = TextureCoordinate()
  n66 = VectorParameter() {group=Distortion, parameterName=Gradient Distortion UV, sortPriority=1, defaultValue=<rgba 2 0.1 0 0>}
  n67 = Multiply(A=@n65, B=AppendVector(A=@n66.R, B=@n66.R))
  n68 = AppendVector(A=@n66.B, B=@n66.B)
  n69 = TextureSampleParameter2D(UVs=Add(A=AppendVector(A=@n64.Gradient Distortion Offset U, B=@n64.Gradient Distortion Offset U), B=Panner(Coordinate=@n67, Speed=@n68))) {
   group=Distortion
   parameterName="Gradient Distortion  Tex"
   samplerType=SAMPLERTYPE_Normal
   sortPriority=0
   texture=&T_Weap_Distort_N_02
  }
  n70 = ComponentMask(None=@n65) {g=true}
  n71 = Multiply(A=ComponentMask(None=@n69.RGB) {g=true, r=true}, B=LinearInterpolate(B=OneMinus(None=@n70), Alpha=Constant() {r=1}) {constA=1, constAlpha=0})
  n72 = ScalarParameter() {defaultValue=0.5, group=Distortion, parameterName=Gradient Distortion Intensity A, sortPriority=2}
  n73 = ComponentMask(None=Add(A=TextureCoordinate(), B=Multiply(A=@n71, B=@n72))) {g=true}
  n74 = LinearInterpolate(
   A=Multiply(A=LinearInterpolate(A=@n54, B=@n62, Alpha=@n63), B=@n73)
   B=@n73
   Alpha=ScalarParameter() {defaultValue=0.3, group=Dissolve, parameterName=Lerp Gradient}
  )
  n75 = ScalarParameter() {defaultValue=1.1, group=Dissolve, parameterName=Dissolve Power}
  n76 = ScalarParameter() {defaultValue=20, group=Dissolve, parameterName=Dissolve Edge Hardness, sliderMin=0.5}
  n77 = ScalarParameter() {defaultValue=0.9, group=Dissolve, parameterName=Dissolve Range}
  n78 = DynamicParameter() {defaultValue=<rgba 0 1 1 0>, paramNames [Random Offset, WPO intensity, Dissolve Range Mul, ""]}
  n79 = Subtract(A=Multiply(A=Power(Base=@n74, Exp=@n75), B=@n76), B=LinearInterpolate(A=@n76, Alpha=Multiply(A=@n77, B=@n78.Dissolve Range Mul)) {constB=-1})
  n80 = Add(
   A=TextureCoordinate()
   B=Multiply(A=@n71, B=ScalarParameter() {defaultValue=3, group=Distortion, parameterName=Gradient Distortion Intensity B, sortPriority=2})
  )
  n81 = Multiply(A=Power(Base=ComponentMask(None=@n80) {g=true}) {constExponent=1}) {constB=10}
  n82 = ScalarParameter() {defaultValue=1, group=Opacity, parameterName=Opacity A}
  n83 = ScalarParameter() {defaultValue=0.7, group=Opacity, parameterName=Opacity B}
  n84 = Multiply(
   A=Multiply(A=Saturate(None=@n79), B=Saturate(None=@n81))
   B=Multiply(
    A=LinearInterpolate(A=@n82, B=@n83, Alpha=@n13.R)
    B=ScalarParameter() {defaultValue=1, group=Opacity, parameterName=Opacity Multiplier, sortPriority=1}
   )
  ) {constB=3}
  n85 = DepthFade(Opacity=@n84, FadeDistance=ScalarParameter() {defaultValue=4, group=Opacity, parameterName=DepthFadeDistance})
  Multiply(
   A=Multiply(A=@n14.A, B=ShadingPathSwitch(Default=@n85, Mobile=@n84) {desc=Migrated from FeatureLevelSwitch})
   B=MaterialFunctionCall("FadeSensitivity (optional)"=ScalarParameter() {defaultValue=0.5, group=Opacity, parameterName=FadeBasedOnViewAngle}) {materialFunction=&FadeBasedOnViewAngle}.Result
  )
 ]
}
