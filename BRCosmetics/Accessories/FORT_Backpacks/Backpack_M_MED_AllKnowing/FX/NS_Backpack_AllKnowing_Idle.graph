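# Niagara idle effect for the AllKnowing backpack cosmetic: a red light, drifting dark smoke sprites and layered smoke meshes.
# A system-level FXStage value scales light color, sprite spawn rate and sprite alpha.
# Six burst mesh particles are partitioned across six renderers by visibility tag, each with its own scale.
# Their dynamic material parameters follow FXStage, two falling as four rise, with some values differing between front end and gameplay.
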NiagaraSystem /BRCosmetics/Accessories/FORT_Backpacks/Backpack_M_MED_AllKnowing/FX/NS_Backpack_AllKnowing_Idle
dependencies {
 usedDataInterfaces [&NiagaraDataInterfaceCurve, &NiagaraDataInterfaceArrayFloat3]
 usedDynamicInputs []
 usedModules [
  &SystemState
  &Timeline
  &self:SystemSpawnScript.SystemScriptSource.SystemScriptGraph.NiagaraNodeAssignment_1.SetVariables_#1
 ]
}
emitters {
 Light {
  emitterUpdateScript [
   &EmitterState(Life Cycle Mode=::NewEnumerator1, Loop Behavior=::NewEnumerator1, Loop Duration Mode=::NewEnumerator1, UseLoopDelay=false)
   &SpawnBurst_Instantaneous(Use Spawn Probability=false, UseLoopCountLimit=false)
  ]
  particleSpawnScript [&InitializeParticle(
    Color=<rgba 20 0 0 1>
    Lifetime=5
    Position Offset=<0 0 8>
    Sprite Size Mode=::NewEnumerator3
    Uniform Sprite Size=5
    UsePositionOffset=true
    Write Material Random=true
   )]
  particleUpdateScript [
   &ParticleState(Kill Particles When Lifetime Has Elapsed=false)
   &FN_Light_Attributes(Create Exponent Attribute=true, Create Radius Attribute=true, Enable Particle Light Enable=true, Exponent=2, Light Enabled=true, Radius=15)
   &ScaleColor(
    Scale RGB=&VectorFromFloat(Value=&FloatFromCurve(
      FloatCurve=DataInterface() {propertyValues {Curve {keys [
          {arriveTangent=2.5116, interpMode=RCIM_Cubic, leaveTangent=2.5116, tangentMode=RCTM_User, time=0, value=0}
          {interpMode=RCIM_Cubic, tangentMode=RCTM_User, time=1, value=1}
         ]}}}
      CurveIndex=$System.FXStage
     ))
    ScaleA=true
    ScaleRGB=true
   )
  ]
  rendererClasses [&NiagaraLightRendererProperties]
  renderers [{rendererClass=&NiagaraLightRendererProperties, rendererIndex=0}]
 }
 Particle_Smoke {
  emitterUpdateScript [
   &EmitterState()
   &SpawnRate(
    SpawnRate=&Multiply_Float(
     A=6
     B=&FloatFromCurve(
      FloatCurve=DataInterface() {propertyValues {Curve {keys [
          {arriveTangent=2.5116, interpMode=RCIM_Cubic, leaveTangent=2.5116, tangentMode=RCTM_User, time=0, value=0}
          {interpMode=RCIM_Cubic, tangentMode=RCTM_User, time=1, value=1}
         ]}}}
      CurveIndex=$System.FXStage
     )
    )
    Use Spawn Probability=false
   )
  ]
  particleSpawnScript [
   &InitializeParticle(
    Color=<rgba 0 0 0 0.035>
    Lifetime Min=2.5
    Lifetime Mode=::NewEnumerator1
    Position Offset=&RandomRangeVector(Minimum=<0 0 10>, Maximum=<0 0 5>)
    Sprite Rotation Mode=::NewEnumerator2
    Sprite Size=<36 32>
    Sprite Size Mode=::NewEnumerator4
    UsePositionOffset=true
    Write Material Random=true
   )
   &ShapeLocation(Handle Radius=0, Large Radius=4, MaskSpawnGroup=false, Override Seed=false, Shape Primitive=::NewEnumerator3, UseOwnerScale=true)
   &CameraOffset(Camera Offset Amount=&RandomRangeFloat(Minimum=5, Maximum=15))
   &AddVelocity(Velocity=&RandomRangeVector(Minimum=<0 0 2>, Maximum=<0 0 1>))
  ]
  particleUpdateScript [
   &ParticleState()
   &SubUVAnimation()
   &ScaleColor(
    Scale Alpha=&FloatFromCurve(FloatCurve=DataInterface() {propertyValues {Curve {keys [<at 0 0>,<at 0.4 1>,<at 0.7 1>,<at 1 0>]}}}, Scale Curve=$System.FXStage)
    ScaleA=true
    ScaleRGB=true
   )
   &SolveForcesAndVelocity(Clamp Velocity=false, Limit Acceleration=false)
  ]
  rendererClasses [&NiagaraSpriteRendererProperties]
  renderers [{rendererClass=&NiagaraSpriteRendererProperties}]
 }
 Smoke_Meshes {
  emitterUpdateScript [
   &EmitterState(Life Cycle Mode=::NewEnumerator1, Loop Behavior=::NewEnumerator1, Loop Duration Mode=::NewEnumerator1, UseLoopDelay=false)
   &SpawnBurst_Instantaneous(Spawn Count=6, Use Spawn Probability=false, UseLoopCountLimit=false)
  ]
  particleSpawnScript [
   &PartitionParticles(Number Of Partitions=6)
   /SetVariables(Particles.VisibilityTag=$Particles.PartitionParticles.Partition) {isSetParameters=true}
   &InitializeParticle(
    Mesh Scale=&SelectVectorFromArray(
     Direct Array Index=$Particles.PartitionParticles.Partition
     Vector Selection Array=DataInterface() {propertyValues {FloatData [<0.4 0.4 0.6>, <0.65 0.65 0.9>, <0.75 0.75 0.8>, <0.8 0.8 1>, <0.7 0.7 1>, <0.45 0.45 1.2>]}}
    )
    Mesh Scale Mode=::NewEnumerator4
    UsePositionOffset=false
    Write Material Random=true
   )
  ]
  particleUpdateScript [
   &ParticleState(Kill Particles When Lifetime Has Elapsed=false, Let Infinitely Lived Particles Die When Emitter Deactivates=true)
   &DynamicMaterialParameters(
    Index 0 Param 1=&RandomRangeFloat(Minimum=0, Maximum=1)
    Index 0 Param 2=&self:FloatfromSpawnGroup(
     Spawn Group 000:NiagaraFloat=1
     Spawn Group 001:NiagaraFloat=1
     Spawn Group 002:NiagaraFloat=&FN_FrontEndFloatModifier(Gameplay=4, Front End=2)
     Spawn Group 003:NiagaraFloat=&FN_FrontEndFloatModifier(Gameplay=4, Front End=2)
     Spawn Group 004:NiagaraFloat=1
     Spawn Group 005:NiagaraFloat=2
     SpawnGroup:NiagaraInt32=$Particles.PartitionParticles.Partition
    )
    Index 0 Param 3=&self:FloatfromSpawnGroup(
     Spawn Group 000:NiagaraFloat=&FloatFromCurve(
      FloatCurve=DataInterface() {propertyValues {Curve {keys [{interpMode=RCIM_Cubic, tangentMode=RCTM_User, time=0, value=1}, {interpMode=RCIM_Cubic, tangentMode=RCTM_User, time=1, value=0}]}}}
      CurveIndex=$System.FXStage
     )
     Spawn Group 001:NiagaraFloat=&FloatFromCurve(
      FloatCurve=DataInterface() {propertyValues {Curve {keys [{interpMode=RCIM_Cubic, tangentMode=RCTM_User, time=0, value=1}, {interpMode=RCIM_Cubic, tangentMode=RCTM_User, time=1, value=0}]}}}
      CurveIndex=$System.FXStage
     )
     Spawn Group 002:NiagaraFloat=&FloatFromCurve(
      FloatCurve=DataInterface() {propertyValues {Curve {keys [
          {arriveTangent=2.5116, interpMode=RCIM_Cubic, leaveTangent=2.5116, tangentMode=RCTM_User, time=0, value=0}
          {interpMode=RCIM_Cubic, tangentMode=RCTM_User, time=1, value=1}
         ]}}}
      CurveIndex=$System.FXStage
     )
     Spawn Group 003:NiagaraFloat=&FloatFromCurve(
      FloatCurve=DataInterface() {propertyValues {Curve {keys [
          {arriveTangent=2.5116, interpMode=RCIM_Cubic, leaveTangent=2.5116, tangentMode=RCTM_User, time=0, value=0}
          {interpMode=RCIM_Cubic, tangentMode=RCTM_User, time=1, value=1}
         ]}}}
      CurveIndex=$System.FXStage
     )
     Spawn Group 004:NiagaraFloat=&FloatFromCurve(
      FloatCurve=DataInterface() {propertyValues {Curve {keys [
          {arriveTangent=2.5116, interpMode=RCIM_Cubic, leaveTangent=2.5116, tangentMode=RCTM_User, time=0, value=0}
          {interpMode=RCIM_Cubic, tangentMode=RCTM_User, time=1, value=1}
         ]}}}
      CurveIndex=$System.FXStage
     )
     Spawn Group 005:NiagaraFloat=&FloatFromCurve(
      FloatCurve=DataInterface() {propertyValues {Curve {keys [
          {arriveTangent=2.5116, interpMode=RCIM_Cubic, leaveTangent=2.5116, tangentMode=RCTM_User, time=0, value=0}
          {interpMode=RCIM_Cubic, tangentMode=RCTM_User, time=1, value=1}
         ]}}}
      CurveIndex=$System.FXStage
     )
     SpawnGroup:NiagaraInt32=$Particles.PartitionParticles.Partition
    )
    Index 1 Param 3=&self:FloatfromSpawnGroup(
     Spawn Group 000:NiagaraFloat=0.1
     Spawn Group 001:NiagaraFloat=0.1
     Spawn Group 002:NiagaraFloat=0.1
     Spawn Group 003:NiagaraFloat=0.1
     Spawn Group 004:NiagaraFloat=0
     Spawn Group 005:NiagaraFloat=0
     SpawnGroup:NiagaraInt32=$Particles.PartitionParticles.Partition
    )
    Module.Index0Param0Write=true
    Module.Index0Param1Write=true
    Module.Index0Param2Write=true
    Module.Index0Param3Write=false
    Module.Index1Param0Write=false
    Module.Index1Param1Write=false
    Module.Index1Param2Write=true
    Module.Index1Param3Write=true
    Write Parameter Index 1=true
   )
  ]
  rendererClasses [&NiagaraMeshRendererProperties]
  renderers [
   {rendererClass=&NiagaraMeshRendererProperties, rendererIndex=0}
   {rendererClass=&NiagaraMeshRendererProperties, rendererIndex=1}
   {rendererClass=&NiagaraMeshRendererProperties, rendererIndex=2}
   {rendererClass=&NiagaraMeshRendererProperties, rendererIndex=3}
   {rendererClass=&NiagaraMeshRendererProperties, rendererIndex=4}
   {rendererClass=&NiagaraMeshRendererProperties, rendererIndex=5}
  ]
 }
}
summary {
 systemName=NS_Backpack_AllKnowing_Idle
 emitters [
  {emitterName=Particle_Smoke, rendererClasses [&NiagaraSpriteRendererProperties]}
  {emitterName=Light, rendererClasses [&NiagaraLightRendererProperties]}
  {emitterName=Smoke_Meshes, rendererClasses [&NiagaraMeshRendererProperties]}
 ]
 userVariables [
  User.bIsFrontEnd {type=NiagaraBool}
  User.bIsFrontEndPreview {type=NiagaraBool}
  User.FX Stage {type=NiagaraFloat, defaultValue=0, description=""}
  User.Opened {type=NiagaraBool}
 ]
}
