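# Animation Blueprint for the AllKnowing backpack, cycling it through closed idle, open, open idle and close.
# bIsInFrontEnd selects between front-end and game state machines; the front-end one starts partway into closed idle, then opens automatically.
# A Random Trigger notify randomly sets CanChangeState, letting closed idle open near its end, and state notifies broadcast Open Event for opening and closing.
# Fixed bone transforms then rotate root and offset pack.
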AnimBlueprint /BRCosmetics/Accessories/FORT_Backpacks/Backpack_M_MED_AllKnowing/Meshes/M_MED_AllKnowing_Pack_AnimBP
target_skeleton=&/M_MED_AllKnowing_Pack_Skeleton
graphs [
 AnimGraph {nodes [
   Input Pose()
   AnimGraphNode_StateMachine_0 = FrontEndStateMachine() {is=title AnimGraphNode_StateMachine}
   AnimGraphNode_StateMachine_1 = GameStateMachine() {is=title AnimGraphNode_StateMachine}
   n1 = Get bIsInFrontEnd()
   n2 = "Transform (Modify) Bone - Bone: root"(
    ComponentPose=Local To Component(LocalPose=Blend Poses by bool(BlendPose_0=@AnimGraphNode_StateMachine_0, BlendPose_1=@AnimGraphNode_StateMachine_1, bActiveValue=@n1) {blendType=Linear})
    Translation="0,0,0"
    Rotation="0,0,-20"
   ) {
    rotationMode=BMM_Additive
    rotationSpace=BCS_BoneSpace
    alphaBoolBlend {blendOption=Linear}
    boneToModify {boneName=root}
    rotation {roll=-20}
   }
   n3 = "Transform (Modify) Bone - Bone: pack"(ComponentPose=@n2, Translation="0,9,4", Rotation="0,0,0") {
    translationMode=BMM_Additive
    translationSpace=BCS_BoneSpace
    alphaBoolBlend {blendOption=Linear}
    boneToModify {boneName=pack}
    translation {y=9, z=4}
   }
   Output Pose(Result=Component To Local(ComponentPose=@n3)) {layerGroup=None}
  ]}
 FrontEndStateMachine {
  entry=ClosedIdleInitiate
  path=AnimGraph.AnimGraphNode_StateMachine_0.FrontEndStateMachine
  ids=id
  nodes [
   AnimStateNode_0 = Closed Idle() {
    is=title AnimStateNode
    stateEntered {guid=#1, notifyName=Entered Closed Idle State}
    stateLeft {guid=#2, notifyName=Exited Closed Idle State}
   }
   AnimStateNode_1 = Open() {is=title AnimStateNode}
   AnimStateNode_2 = Open Idle() {is=title AnimStateNode}
   AnimStateNode_3 = Close() {is=title AnimStateNode}
   AnimStateNode_4 = ClosedIdleInitiate() {is=title AnimStateNode}
   AnimStateTransitionNode_0 = AnimStateTransitionNode()
   AnimStateTransitionNode_1 = AnimStateTransitionNode() {bAutomaticRuleBasedOnSequencePlayerInState=true}
   AnimStateTransitionNode_2 = AnimStateTransitionNode() {bAutomaticRuleBasedOnSequencePlayerInState=true}
   AnimStateTransitionNode_3 = AnimStateTransitionNode() {bAutomaticRuleBasedOnSequencePlayerInState=true}
   AnimStateTransitionNode_4 = AnimStateTransitionNode() {bAutomaticRuleBasedOnSequencePlayerInState=true}
  ]
  links [
   {
    dst=Closed Idle
    rule=AnimGraph.AnimGraphNode_StateMachine_0.FrontEndStateMachine.AnimStateTransitionNode_3.Transition
    src=Close
    via=AnimStateTransitionNode_3
   }
   {
    dst=Open
    rule=AnimGraph.AnimGraphNode_StateMachine_0.FrontEndStateMachine.AnimStateTransitionNode_0.Transition
    src=Closed Idle
    via=AnimStateTransitionNode_0
   }
   {
    dst=Open
    rule=AnimGraph.AnimGraphNode_StateMachine_0.FrontEndStateMachine.AnimStateTransitionNode_4.Transition
    src=ClosedIdleInitiate
    via=AnimStateTransitionNode_4
   }
   {
    dst=Open Idle
    rule=AnimGraph.AnimGraphNode_StateMachine_0.FrontEndStateMachine.AnimStateTransitionNode_1.Transition
    src=Open
    via=AnimStateTransitionNode_1
   }
   {
    dst=Close
    rule=AnimGraph.AnimGraphNode_StateMachine_0.FrontEndStateMachine.AnimStateTransitionNode_2.Transition
    src=Open Idle
    via=AnimStateTransitionNode_2
   }
  ]
 }
 Closed Idle {
  path=AnimGraph.AnimGraphNode_StateMachine_0.FrontEndStateMachine.AnimStateNode_0.Closed Idle
  nodes [Output Animation Pose(Result=Sequence Player 'AllKnowing_ClosedIdle_Pack'() {sequence=/BRCosmetics/Accessories/FORT_Backpacks/Backpack_M_MED_AllKnowing/Animation/AllKnowing_ClosedIdle_Pack})]
 }
 Open {
  path=AnimGraph.AnimGraphNode_StateMachine_0.FrontEndStateMachine.AnimStateNode_1.Open
  nodes [Output Animation Pose(Result=Sequence Player 'AllKnowing_Open_Pack'!(bLoopAnimation:bool=False) -> (Pose) {bLoopAnimation=false, sequence=/BRCosmetics/Accessories/FORT_Backpacks/Backpack_M_MED_AllKnowing/Animation/AllKnowing_Open_Pack})]
 }
 Open Idle {
  path=AnimGraph.AnimGraphNode_StateMachine_0.FrontEndStateMachine.AnimStateNode_2.Open Idle
  nodes [Output Animation Pose(Result=Sequence Player 'AllKnowing_Openidle_Pack'!(bLoopAnimation:bool=False) -> (Pose) {bLoopAnimation=false, sequence=/BRCosmetics/Accessories/FORT_Backpacks/Backpack_M_MED_AllKnowing/Animation/AllKnowing_Openidle_Pack})]
 }
 Close {
  path=AnimGraph.AnimGraphNode_StateMachine_0.FrontEndStateMachine.AnimStateNode_3.Close
  nodes [Output Animation Pose(Result=Sequence Player 'AllKnowing_Closed_Pack'!(bLoopAnimation:bool=False) -> (Pose) {bLoopAnimation=false, sequence=/BRCosmetics/Accessories/FORT_Backpacks/Backpack_M_MED_AllKnowing/Animation/AllKnowing_Closed_Pack})]
 }
 ClosedIdleInitiate {
  path=AnimGraph.AnimGraphNode_StateMachine_0.FrontEndStateMachine.AnimStateNode_4.ClosedIdleInitiate
  nodes [Output Animation Pose(Result=Sequence Player 'AllKnowing_ClosedIdle_Pack'!(StartPosition:float=3.5) -> (Pose) {sequence=&AllKnowing_ClosedIdle_Pack})]
 }
 Transition {
  path=AnimGraph.AnimGraphNode_StateMachine_0.FrontEndStateMachine.AnimStateTransitionNode_0.Transition
  nodes [
   n1 = "Get Relevant Anim Time Remaining (Closed Idle)"!() -> (ReturnValue:float) {is=title K2Node_AnimGetter}
   Result(bCanEnterTransition=AND Boolean(A=Get CanChangeState() -> (CanChangeState:bool), B="float <= float"(A=@n1, B=0.2) -> (ReturnValue=~)))
  ]
 }
 Transition {path=AnimGraph.AnimGraphNode_StateMachine_0.FrontEndStateMachine.AnimStateTransitionNode_1.Transition, nodes [Result()]}
 Transition {path=AnimGraph.AnimGraphNode_StateMachine_0.FrontEndStateMachine.AnimStateTransitionNode_2.Transition, nodes [Result()]}
 Transition {path=AnimGraph.AnimGraphNode_StateMachine_0.FrontEndStateMachine.AnimStateTransitionNode_3.Transition, nodes [Result()]}
 Transition {path=AnimGraph.AnimGraphNode_StateMachine_0.FrontEndStateMachine.AnimStateTransitionNode_4.Transition, nodes [Result()]}
 GameStateMachine {
  entry=Closed Idle
  path=AnimGraph.AnimGraphNode_StateMachine_1.GameStateMachine
  ids=id
  nodes [
   AnimStateNode_0 = Closed Idle() {
    stateEntered {guid=#1, notifyName=Entered Closed Idle State}
    stateLeft {guid=#2, notifyName=Exited Closed Idle State}
   }
   AnimStateNode_1 = Open()
   AnimStateNode_2 = Open Idle()
   AnimStateNode_3 = Close()
   AnimStateTransitionNode_0 = AnimStateTransitionNode() {transitionStart {guid=#3, notifyName=Leaving Closed Idle}}
   AnimStateTransitionNode_1 = AnimStateTransitionNode() {bAutomaticRuleBasedOnSequencePlayerInState=true}
   AnimStateTransitionNode_2 = AnimStateTransitionNode() {bAutomaticRuleBasedOnSequencePlayerInState=true, transitionStart {guid=#4, notifyName=Leaving Open Idle}}
   AnimStateTransitionNode_3 = AnimStateTransitionNode() {bAutomaticRuleBasedOnSequencePlayerInState=true}
  ]
  links [
   {
    dst=Closed Idle
    rule=AnimGraph.AnimGraphNode_StateMachine_1.GameStateMachine.AnimStateTransitionNode_3.Transition
    src=Close
    via=AnimStateTransitionNode_3
   }
   {
    dst=Open
    rule=AnimGraph.AnimGraphNode_StateMachine_1.GameStateMachine.AnimStateTransitionNode_0.Transition
    src=Closed Idle
    via=AnimStateTransitionNode_0
   }
   {
    dst=Open Idle
    rule=AnimGraph.AnimGraphNode_StateMachine_1.GameStateMachine.AnimStateTransitionNode_1.Transition
    src=Open
    via=AnimStateTransitionNode_1
   }
   {
    dst=Close
    rule=AnimGraph.AnimGraphNode_StateMachine_1.GameStateMachine.AnimStateTransitionNode_2.Transition
    src=Open Idle
    via=AnimStateTransitionNode_2
   }
  ]
 }
 Closed Idle {
  path=AnimGraph.AnimGraphNode_StateMachine_1.GameStateMachine.AnimStateNode_0.Closed Idle
  nodes [Output Animation Pose(Result=Sequence Player 'AllKnowing_ClosedIdle_Pack'!() -> (Pose) {sequence=&AllKnowing_ClosedIdle_Pack})]
 }
 Open {
  path=AnimGraph.AnimGraphNode_StateMachine_1.GameStateMachine.AnimStateNode_1.Open
  nodes [Output Animation Pose(Result=Sequence Player 'AllKnowing_Open_Pack'(bLoopAnimation=False) {bLoopAnimation=false, sequence=&AllKnowing_Open_Pack})]
 }
 Open Idle {
  path=AnimGraph.AnimGraphNode_StateMachine_1.GameStateMachine.AnimStateNode_2.Open Idle
  nodes [Output Animation Pose(Result=Sequence Player 'AllKnowing_Openidle_Pack'(bLoopAnimation=False) {bLoopAnimation=false, sequence=&AllKnowing_Openidle_Pack})]
 }
 Close {
  path=AnimGraph.AnimGraphNode_StateMachine_1.GameStateMachine.AnimStateNode_3.Close
  nodes [Output Animation Pose(Result=Sequence Player 'AllKnowing_Closed_Pack'(bLoopAnimation=False) {bLoopAnimation=false, sequence=&AllKnowing_Closed_Pack})]
 }
 Transition {
  path=AnimGraph.AnimGraphNode_StateMachine_1.GameStateMachine.AnimStateTransitionNode_0.Transition
  nodes [
   n1 = "Get Relevant Anim Time Remaining (Closed Idle)"()
   Result(bCanEnterTransition=AND Boolean(A=Get CanChangeState(), B="float <= float"(A=@n1, B=0.2) -> (ReturnValue=~)))
  ]
 }
 Transition {path=AnimGraph.AnimGraphNode_StateMachine_1.GameStateMachine.AnimStateTransitionNode_1.Transition, nodes [Result()]}
 Transition {path=AnimGraph.AnimGraphNode_StateMachine_1.GameStateMachine.AnimStateTransitionNode_2.Transition, nodes [Result()]}
 Transition {path=AnimGraph.AnimGraphNode_StateMachine_1.GameStateMachine.AnimStateTransitionNode_3.Transition, nodes [Result()]}
 EventGraph {nodes [
   AnimNotify_Random Trigger() {is=title K2Node_Event}
   > Set CanChangeState(CanChangeState:bool=RandomBool!() -> (ReturnValue:bool)) -> (Output_Get:bool)
   AnimNotify_Exited Closed Idle State() {is=title K2Node_Event}
   > Set CanChangeState(CanChangeState=false)
   > Call Open Event!(execute:Exec, self:M MED All Knowing Pack Anim BP Object Reference, Opening:bool=true) -> (then:Exec) {is=title K2Node_CallDelegate}
   AnimNotify_Leaving Open Idle() {is=title K2Node_Event}
   > Call Open Event()
   TryGetPawnOwner()
   Event BlueprintUpdateAnimation()
  ]}
 Open Event {nodes [Open Event!() -> (then, Opening:bool=~) {is=title K2Node_FunctionEntry}]}
]
variables [
 CanChangeState {type=boolean}
 Open Event {default=()}
]
